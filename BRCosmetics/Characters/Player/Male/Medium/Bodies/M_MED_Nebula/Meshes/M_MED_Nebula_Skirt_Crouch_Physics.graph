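# Physics asset for the crouch variant of the M_MED_Nebula skirt, covering skirt panels, bows and ropes.
# Capsules on pelvis, spine, thighs and calves form the body volume.
# Upper skirt, bow and rope chains hang from tiny MSTR anchor bodies with free angular joints to the pelvis.
# Lower skirt panels, mostly boxes, attach to thighs or pelvis, and cloth chain links use narrow swing limits with zero twist.
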PhysicsAsset /BRCosmetics/Characters/Player/Male/Medium/Bodies/M_MED_Nebula/Meshes/M_MED_Nebula_Skirt_Crouch_Physics
bodies {
 calf_l [{shapeType=Capsule, length=26.6325, radius=5.9116, center=<-16.4232 0 0>, rotation=<rot 90 0 0>}]
 calf_r [{shapeType=Capsule, length=26.6325, radius=5.9116, center=<16.4232 0 0>, rotation=<rot 90 0 180>}]
 dyn_bow_l_01 [{
   shapeType=Capsule
   length=5.1561
   radius=4.6517
   shapeName=dyn_bow_l_01_capsule
   center=<4.8892 1.8 0>
   rotation=<rot 55.1881 -0.0001 9.5228>
  }]
 dyn_bow_l_MSTR [{shapeType=Capsule, length=0.505, radius=0.505, shapeName=dyn_bow_l_MSTR_capsule, rotation=<rot 0 -90 0>}]
 dyn_bow_r_01 [{shapeType=Capsule, length=0, radius=3.369, shapeName=dyn_bow_r_01_capsule, center=<3.4889 0 0>}]
 dyn_bow_r_MSTR [{shapeType=Capsule, length=0.505, radius=0.505, shapeName=dyn_bow_r_MSTR_capsule, rotation=<rot 0 -90 0>}]
 dyn_rope_l_01 [{shapeType=Capsule, length=5.7435, radius=1.3163, shapeName=dyn_rope_l_01_capsule, center=<4.894 0 0>, rotation=<rot 90 0 0>}]
 dyn_rope_l_02 [{shapeType=Capsule, length=7.4017, radius=1.3801, shapeName=dyn_rope_l_02_capsule, center=<4.894 0 0>, rotation=<rot 90 0 0>}]
 dyn_rope_l_03 [{shapeType=Capsule, length=8.1334, radius=1.6569, shapeName=dyn_rope_l_03_capsule, center=<4.894 0 0>, rotation=<rot 90 0 0>}]
 dyn_rope_l_MSTR [{shapeType=Capsule, length=0.505, radius=0.505, shapeName=dyn_rope_l_MSTR_capsule, rotation=<rot 0 -90 0>}]
 dyn_rope_r_01 [{shapeType=Capsule, length=5.7435, radius=1.3145, shapeName=dyn_rope_r_01_capsule, center=<4.894 0 0>, rotation=<rot 90 0 0>}]
 dyn_rope_r_02 [{shapeType=Capsule, length=7.4017, radius=1.375, shapeName=dyn_rope_r_02_capsule, center=<4.894 0 0>, rotation=<rot 90 0 0>}]
 dyn_rope_r_03 [{shapeType=Capsule, length=7.7785, radius=2.0054, shapeName=dyn_rope_r_03_capsule, center=<4.894 0 0>, rotation=<rot 90 0 0>}]
 dyn_rope_r_MSTR [{shapeType=Capsule, length=0.505, radius=0.505, shapeName=dyn_rope_r_MSTR_capsule, rotation=<rot 0 -90 0>}]
 dyn_skirtBtA_l_01 [{
   shapeType=Box
   extentX=7.862
   extentY=14.4675
   extentZ=2.038
   shapeName=dyn_skirtBtA_l_01_box
   center=<2.8371 0 0>
   rotation=<rot 0.0001 0 -1.0549>
  }]
 dyn_skirtBtA_l_02 [{
   shapeType=Box
   extentX=13.1946
   extentY=12.2254
   extentZ=2.0303
   shapeName=dyn_skirtBtA_l_02_box
   center=<3.8664 0 0>
   rotation=<rot -4.5034 0.5382 -6.8225>
  }]
 dyn_skirtBtA_r_01 [{
   shapeType=Box
   extentX=7.5
   extentY=14.5396
   extentZ=2.038
   shapeName=dyn_skirtBtA_r_01_box
   center=<2.8371 0 0>
   rotation=<rot 0.0142 -0.0038 11.1079>
  }]
 dyn_skirtBtA_r_02 [{
   shapeType=Box
   extentX=13.869
   extentY=11.94
   extentZ=2.038
   shapeName=dyn_skirtBtA_r_02_box
   center=<5.7158 0 0>
   rotation=<rot 0.0001 0 11.4521>
  }]
 dyn_skirtBtB_l_01 [{
   shapeType=Box
   extentX=7.5
   extentY=20.7693
   extentZ=2.038
   shapeName=dyn_skirtBtB_l_01_box
   center=<2.8371 0 0>
   rotation=<rot 0 0.0002 24.0162>
  }]
 dyn_skirtBtB_l_02 [{
   shapeType=Box
   extentX=7.5
   extentY=20.7003
   extentZ=2.038
   shapeName=dyn_skirtBtB_l_02_box
   center=<2.8371 0 0>
   rotation=<rot 0 0.0002 24.0162>
  }]
 dyn_skirtBtB_r_01 [{shapeType=Box, extentX=7.5, extentY=23.2843, extentZ=2.038, shapeName=dyn_skirtBtB_r_01_box, center=<2.8371 0 0>}]
 dyn_skirtBtB_r_02 [{
   shapeType=Capsule
   length=17.8126
   radius=4.1917
   shapeName=dyn_skirtBtB_r_02_capsule
   center=<2.9093 0 2.5>
   rotation=<rot -0.0005 29.0932 90.0009>
  }]
 dyn_skirtBtC_l_01 [{shapeType=Box, extentX=7.5, extentY=13.5432, extentZ=2.038, shapeName=dyn_skirtBtC_l_01_box, center=<2.8371 0 0>}]
 dyn_skirtBtC_l_02 [{
   shapeType=Capsule
   length=17.2372
   radius=3.5627
   shapeName=dyn_skirtBtC_l_02_capsule
   center=<2.6316 0 0>
   rotation=<rot 89.5756 90.0331 179.9966>
  }]
 dyn_skirtBtC_r_01 [{
   shapeType=Box
   extentX=7.5
   extentY=15.9068
   extentZ=2.038
   shapeName=dyn_skirtBtC_r_01_box
   center=<2.8371 0 0>
   rotation=<rot 0 0 -15.5008>
  }]
 dyn_skirtBtC_r_02 [{
   shapeType=Capsule
   length=17.2372
   radius=3.5627
   shapeName=dyn_skirtBtC_r_02_capsule
   center=<2.6316 0 0>
   rotation=<rot 76.3234 -89.9995 0.0006>
  }]
 dyn_skirtTpA_l_01 [{shapeType=Capsule, length=8.9338, radius=2.3821, shapeName=dyn_skirtTpA_l_01_capsule, center=<5.2991 0 3.7>, rotation=<rot 90 0 0>}]
 dyn_skirtTpA_l_02 [{shapeType=Capsule, length=7.699, radius=3.617, shapeName=dyn_skirtTpA_l_02_capsule, center=<5.2991 -1.9 3.7>, rotation=<rot 90 0 0>}]
 dyn_skirtTpA_l_03 [{shapeType=Capsule, length=8.9585, radius=3.2312, shapeName=dyn_skirtTpA_l_03_capsule, center=<5.2991 -1.5 3.7>, rotation=<rot 90 0 0>}]
 dyn_skirtTpA_l_MSTR [{shapeType=Capsule, length=0.5, radius=0.5, shapeName=dyn_skirtTpA_l_MSTR_capsule}]
 dyn_skirtTpA_r_01 [{shapeType=Capsule, length=8.8336, radius=2.4824, shapeName=dyn_skirtTpA_r_01_capsule, center=<5.2991 0 3.7>, rotation=<rot 90 0 0>}]
 dyn_skirtTpA_r_02 [{shapeType=Capsule, length=7.7248, radius=3.5911, shapeName=dyn_skirtTpA_r_02_capsule, center=<5.2991 0 3.7>, rotation=<rot 90 0 0>}]
 dyn_skirtTpA_r_03 [{shapeType=Capsule, length=7.4965, radius=3.8194, shapeName=dyn_skirtTpA_r_03_capsule, center=<5.2991 0 3.7>, rotation=<rot 90 0 0>}]
 dyn_skirtTpA_r_MSTR [{shapeType=Capsule, length=0.5, radius=0.5, shapeName=dyn_skirtTpA_r_MSTR_capsule}]
 dyn_skirtTpB_l_01 [{
   shapeType=Capsule
   length=5.2819
   radius=2.93
   shapeName=dyn_skirtTpB_l_01_capsule
   center=<5.2991 0 3.6>
   rotation=<rot 86.6456 -0.0003 -0.0003>
  }]
 dyn_skirtTpB_l_02 [{
   shapeType=Capsule
   length=14.6778
   radius=3.5106
   shapeName=dyn_skirtTpB_l_02_capsule
   center=<5.1838 0 1.7>
   rotation=<rot 82.9002 -0.0003 -0.0003>
  }]
 dyn_skirtTpB_l_MSTR [{shapeType=Capsule, length=0.5, radius=0.5, shapeName=dyn_skirtTpB_l_MSTR_capsule}]
 dyn_skirtTpB_r_01 [{shapeType=Capsule, length=8.3231, radius=1.2285, shapeName=dyn_skirtTpB_r_01_capsule, center=<5.2991 0 2.2>, rotation=<rot 90 0 0>}]
 dyn_skirtTpB_r_02 [{shapeType=Capsule, length=9.2929, radius=3.4827, shapeName=dyn_skirtTpB_r_02_capsule, center=<5.2991 0 2.2>, rotation=<rot 67.7815 0 0.0001>}]
 dyn_skirtTpB_r_MSTR [{shapeType=Capsule, length=0.5, radius=0.5, shapeName=dyn_skirtTpB_r_MSTR_capsule}]
 dyn_skirtTpC_l_01 [{shapeType=Capsule, length=9.6148, radius=1.7011, shapeName=dyn_skirtTpC_l_01_capsule, center=<5.2991 0 2.2>, rotation=<rot 90 0 0>}]
 dyn_skirtTpC_l_02 [{shapeType=Capsule, length=9.7307, radius=3.6248, shapeName=dyn_skirtTpC_l_02_capsule, center=<5.2991 0 2.2>, rotation=<rot 72.4591 0 0>}]
 dyn_skirtTpC_l_MSTR [{shapeType=Capsule, length=0.5, radius=0.5, shapeName=dyn_skirtTpC_l_MSTR_capsule}]
 dyn_skirtTpC_r_01 [{shapeType=Capsule, length=9.2131, radius=2.1028, shapeName=dyn_skirtTpC_r_01_capsule, center=<5.2991 0 3>, rotation=<rot 90 0 0>}]
 dyn_skirtTpC_r_02 [{shapeType=Capsule, length=14.2001, radius=3.6028, shapeName=dyn_skirtTpC_r_02_capsule, center=<8.6086 0 4.0475>, rotation=<rot 78.9735 0 0>}]
 dyn_skirtTpC_r_MSTR [{shapeType=Capsule, length=0.5, radius=0.5, shapeName=dyn_skirtTpC_r_MSTR_capsule}]
 dyn_skirtTp_c_01 [{shapeType=Capsule, length=6.9365, radius=3.3693, shapeName=dyn_skirtTp_c_01_capsule, center=<8.2662 0 2.2>, rotation=<rot 90 0 0>}]
 dyn_skirtTp_c_02 [{shapeType=Capsule, length=9.0704, radius=4.9943, shapeName=dyn_skirtTp_c_02_capsule, center=<8.1286 0 5.3>, rotation=<rot 90 0 0>}]
 dyn_skirtTp_c_03 [{shapeType=Capsule, length=9.6994, radius=4.7682, shapeName=dyn_skirtTp_c_03_capsule, center=<6.4026 0 2.2>, rotation=<rot 90 0 0>}]
 dyn_skirtTp_c_04 [{shapeType=Capsule, length=6.4258, radius=4.2929, shapeName=dyn_skirtTp_c_04_capsule, center=<9.9735 0 2.2>, rotation=<rot 90 0 0>}]
 dyn_skirtTp_c_MSTR [{shapeType=Capsule, length=0.5, radius=0.5, shapeName=dyn_skirtTp_c_MSTR_capsule}]
 pelvis [
  {shapeType=Capsule, length=10.7635, radius=10.8182}
  {shapeType=Capsule, length=17.3514, radius=9.7304, shapeName=pelvis_capsule, center=<-7.5364 3.7324 0>, rotation=<rot 90 -167.0926 -154.762>}
  {shapeType=Capsule, length=11.521, radius=10.8075, shapeName=pelvis_capsule, center=<-4.4041 -1.861 0>}
 ]
 spine_01 [{shapeType=Capsule, length=6.1459, radius=9.0482, shapeName=spine_01_capsule, center=<0.8123 0 0>, rotation=<rot 0 -90 0>}]
 spine_02 [{shapeType=Capsule, length=3, radius=9.1435}]
 spine_04 [{shapeType=Capsule, length=8.0868, radius=9.1435}]
 spine_05 [{shapeType=Capsule, length=12.1855, radius=9.1435}]
 thigh_l [{shapeType=Capsule, length=31.8341, radius=8.1657, shapeName=thigh_l_capsule, center=<-14.0998 0 0>, rotation=<rot 90 0 0>}]
 thigh_r [{shapeType=Capsule, length=31.2752, radius=8.7246, shapeName=thigh_l_capsule, center=<14.0998 0 0>, rotation=<rot 90 0 180>}]
}
constraints [
 {bone1Name=spine_02, bone2Name=spine_01, swing1LimitDegrees=45, swing2LimitDegrees=45, twistLimitDegrees=45}
 {bone1Name=spine_04, bone2Name=spine_02, swing1LimitDegrees=45, swing2LimitDegrees=45, twistLimitDegrees=45}
 {bone1Name=spine_05, bone2Name=spine_04, swing1LimitDegrees=45, swing2LimitDegrees=45, twistLimitDegrees=45}
 {bone1Name=calf_l, bone2Name=thigh_l, swing1LimitDegrees=45, swing2LimitDegrees=45, twistLimitDegrees=45}
 {
  bone1Name=thigh_l
  bone2Name=pelvis
  swing1LimitDegrees=45
  swing1Motion=Free
  swing2LimitDegrees=45
  swing2Motion=Free
  twistLimitDegrees=45
  twistMotion=Free
 }
 {bone1Name=calf_r, bone2Name=thigh_r, swing1LimitDegrees=45, swing2LimitDegrees=45, twistLimitDegrees=45}
 {bone1Name=dyn_rope_r_03, bone2Name=dyn_rope_r_02, swing1LimitDegrees=20, swing2LimitDegrees=20, twistLimitDegrees=0}
 {
  bone1Name=dyn_bow_l_MSTR
  bone2Name=pelvis
  swing1LimitDegrees=45
  swing1Motion=Free
  swing2LimitDegrees=45
  swing2Motion=Free
  twistLimitDegrees=45
  twistMotion=Free
 }
 {bone1Name=dyn_bow_l_01, bone2Name=dyn_bow_l_MSTR, swing1LimitDegrees=10, swing2LimitDegrees=10, twistLimitDegrees=0}
 {
  bone1Name=dyn_bow_r_MSTR
  bone2Name=pelvis
  swing1LimitDegrees=45
  swing1Motion=Free
  swing2LimitDegrees=45
  swing2Motion=Free
  twistLimitDegrees=45
  twistMotion=Free
 }
 {bone1Name=dyn_bow_r_01, bone2Name=dyn_bow_r_MSTR, swing1LimitDegrees=10, swing2LimitDegrees=10, twistLimitDegrees=0}
 {
  bone1Name=dyn_rope_l_MSTR
  bone2Name=pelvis
  swing1LimitDegrees=45
  swing1Motion=Free
  swing2LimitDegrees=45
  swing2Motion=Free
  twistLimitDegrees=45
  twistMotion=Free
 }
 {bone1Name=dyn_rope_l_01, bone2Name=dyn_rope_l_MSTR, swing1LimitDegrees=20, swing2LimitDegrees=20, twistLimitDegrees=0}
 {bone1Name=dyn_rope_l_02, bone2Name=dyn_rope_l_01, swing1LimitDegrees=20, swing2LimitDegrees=20, twistLimitDegrees=0}
 {bone1Name=dyn_rope_l_03, bone2Name=dyn_rope_l_02, swing1LimitDegrees=20, swing2LimitDegrees=20, twistLimitDegrees=0}
 {
  bone1Name=dyn_rope_r_MSTR
  bone2Name=pelvis
  swing1LimitDegrees=45
  swing1Motion=Free
  swing2LimitDegrees=45
  swing2Motion=Free
  twistLimitDegrees=45
  twistMotion=Free
 }
 {bone1Name=dyn_rope_r_01, bone2Name=dyn_rope_r_MSTR, swing1LimitDegrees=20, swing2LimitDegrees=20, twistLimitDegrees=0}
 {bone1Name=dyn_rope_r_02, bone2Name=dyn_rope_r_01, swing1LimitDegrees=20, swing2LimitDegrees=20, twistLimitDegrees=0}
 {
  bone1Name=dyn_skirtTpA_l_MSTR
  bone2Name=pelvis
  swing1LimitDegrees=45
  swing1Motion=Free
  swing2LimitDegrees=45
  swing2Motion=Free
  twistLimitDegrees=45
  twistMotion=Free
 }
 {bone1Name=dyn_skirtTpA_l_01, bone2Name=dyn_skirtTpA_l_MSTR, swing1LimitDegrees=20, swing2LimitDegrees=60, twistLimitDegrees=0}
 {bone1Name=dyn_skirtTpA_l_02, bone2Name=dyn_skirtTpA_l_01, swing1LimitDegrees=20, swing2LimitDegrees=20, twistLimitDegrees=0}
 {bone1Name=dyn_skirtTpA_l_03, bone2Name=dyn_skirtTpA_l_02, swing1LimitDegrees=20, swing2LimitDegrees=20, twistLimitDegrees=0}
 {
  bone1Name=dyn_skirtTpA_r_MSTR
  bone2Name=pelvis
  swing1LimitDegrees=45
  swing1Motion=Free
  swing2LimitDegrees=45
  swing2Motion=Free
  twistLimitDegrees=45
  twistMotion=Free
 }
 {bone1Name=dyn_skirtTpA_r_01, bone2Name=dyn_skirtTpA_r_MSTR, swing1LimitDegrees=20, swing2LimitDegrees=60, twistLimitDegrees=0}
 {bone1Name=dyn_skirtTpA_r_02, bone2Name=dyn_skirtTpA_r_01, swing1LimitDegrees=20, swing2LimitDegrees=20, twistLimitDegrees=0}
 {bone1Name=dyn_skirtTpA_r_03, bone2Name=dyn_skirtTpA_r_02, swing1LimitDegrees=20, swing2LimitDegrees=20, twistLimitDegrees=0}
 {
  bone1Name=dyn_skirtTpB_l_MSTR
  bone2Name=pelvis
  swing1LimitDegrees=45
  swing1Motion=Free
  swing2LimitDegrees=45
  swing2Motion=Free
  twistLimitDegrees=45
  twistMotion=Free
 }
 {bone1Name=dyn_skirtTpB_l_01, bone2Name=dyn_skirtTpB_l_MSTR, swing1LimitDegrees=20, swing2LimitDegrees=20, twistLimitDegrees=0}
 {bone1Name=dyn_skirtTpB_l_02, bone2Name=dyn_skirtTpB_l_01, swing1LimitDegrees=20, swing2LimitDegrees=20, twistLimitDegrees=0}
 {
  bone1Name=dyn_skirtTpB_r_MSTR
  bone2Name=pelvis
  swing1LimitDegrees=45
  swing1Motion=Free
  swing2LimitDegrees=45
  swing2Motion=Free
  twistLimitDegrees=45
  twistMotion=Free
 }
 {bone1Name=dyn_skirtTpB_r_01, bone2Name=dyn_skirtTpB_r_MSTR, swing1LimitDegrees=20, swing2LimitDegrees=20, twistLimitDegrees=0}
 {bone1Name=dyn_skirtTpB_r_02, bone2Name=dyn_skirtTpB_r_01, swing1LimitDegrees=20, swing2LimitDegrees=20, twistLimitDegrees=0}
 {
  bone1Name=dyn_skirtTpC_l_MSTR
  bone2Name=pelvis
  swing1LimitDegrees=45
  swing1Motion=Free
  swing2LimitDegrees=45
  swing2Motion=Free
  twistLimitDegrees=45
  twistMotion=Free
 }
 {bone1Name=dyn_skirtTpC_l_01, bone2Name=dyn_skirtTpC_l_MSTR, swing1LimitDegrees=10, swing2LimitDegrees=20, twistLimitDegrees=0}
 {bone1Name=dyn_skirtTpC_l_02, bone2Name=dyn_skirtTpC_l_01, swing1LimitDegrees=20, swing2LimitDegrees=20, twistLimitDegrees=0}
 {
  bone1Name=dyn_skirtTpC_r_MSTR
  bone2Name=pelvis
  swing1LimitDegrees=45
  swing1Motion=Free
  swing2LimitDegrees=45
  swing2Motion=Free
  twistLimitDegrees=45
  twistMotion=Free
 }
 {bone1Name=dyn_skirtTpC_r_01, bone2Name=dyn_skirtTpC_r_MSTR, swing1LimitDegrees=10, swing2LimitDegrees=20, twistLimitDegrees=0}
 {bone1Name=dyn_skirtTpC_r_02, bone2Name=dyn_skirtTpC_r_01, swing1LimitDegrees=20, swing2LimitDegrees=20, twistLimitDegrees=0}
 {
  bone1Name=dyn_skirtTp_c_MSTR
  bone2Name=pelvis
  swing1LimitDegrees=45
  swing1Motion=Free
  swing2LimitDegrees=45
  swing2Motion=Free
  twistLimitDegrees=45
  twistMotion=Free
 }
 {bone1Name=dyn_skirtTp_c_01, bone2Name=dyn_skirtTp_c_MSTR, swing1LimitDegrees=10, swing2LimitDegrees=14, twistLimitDegrees=0}
 {bone1Name=dyn_skirtTp_c_02, bone2Name=dyn_skirtTp_c_01, swing1LimitDegrees=10, swing2LimitDegrees=15, twistLimitDegrees=0}
 {bone1Name=dyn_skirtTp_c_03, bone2Name=dyn_skirtTp_c_02, swing1LimitDegrees=10, swing2LimitDegrees=15, twistLimitDegrees=0}
 {bone1Name=dyn_skirtTp_c_04, bone2Name=dyn_skirtTp_c_03, swing1LimitDegrees=10, swing2LimitDegrees=20, twistLimitDegrees=0}
 {bone1Name=dyn_skirtBtC_r_02, bone2Name=dyn_skirtBtC_r_01, swing1LimitDegrees=20, swing2LimitDegrees=20, twistLimitDegrees=0}
 {bone1Name=dyn_skirtBtA_l_01, bone2Name=thigh_l, swing1LimitDegrees=20, swing2LimitDegrees=20, twistLimitDegrees=0}
 {bone1Name=dyn_skirtBtA_l_02, bone2Name=dyn_skirtBtA_l_01, swing1LimitDegrees=20, swing2LimitDegrees=20, twistLimitDegrees=0}
 {bone1Name=dyn_skirtBtB_l_01, bone2Name=thigh_l, swing1LimitDegrees=20, swing2LimitDegrees=20, twistLimitDegrees=0}
 {bone1Name=dyn_skirtBtB_l_02, bone2Name=dyn_skirtBtB_l_01, swing1LimitDegrees=20, swing2LimitDegrees=20, twistLimitDegrees=0}
 {bone1Name=dyn_skirtBtC_l_01, bone2Name=pelvis, swing1LimitDegrees=20, swing2LimitDegrees=20, twistLimitDegrees=0}
 {bone1Name=dyn_skirtBtC_l_02, bone2Name=dyn_skirtBtC_l_01, swing1LimitDegrees=20, swing2LimitDegrees=20, twistLimitDegrees=0}
 {bone1Name=dyn_skirtBtA_r_01, bone2Name=thigh_r, swing1LimitDegrees=20, swing2LimitDegrees=20, twistLimitDegrees=0}
 {bone1Name=dyn_skirtBtA_r_02, bone2Name=dyn_skirtBtA_r_01, swing1LimitDegrees=20, swing2LimitDegrees=20, twistLimitDegrees=0}
 {bone1Name=dyn_skirtBtB_r_01, bone2Name=thigh_r, swing1LimitDegrees=20, swing2LimitDegrees=20, twistLimitDegrees=0}
 {bone1Name=dyn_skirtBtB_r_02, bone2Name=dyn_skirtBtB_r_01, swing1LimitDegrees=20, swing2LimitDegrees=20, twistLimitDegrees=0}
 {bone1Name=dyn_skirtBtC_r_01, bone2Name=pelvis, swing1LimitDegrees=20, swing2LimitDegrees=20, twistLimitDegrees=0}
 {
  bone1Name=thigh_r
  bone2Name=pelvis
  swing1LimitDegrees=45
  swing1Motion=Free
  swing2LimitDegrees=45
  swing2Motion=Free
  twistLimitDegrees=45
  twistMotion=Free
 }
 {
  bone1Name=spine_01
  bone2Name=pelvis
  swing1LimitDegrees=45
  swing1Motion=Free
  swing2LimitDegrees=45
  swing2Motion=Free
  twistLimitDegrees=45
  twistMotion=Free
 }
]
