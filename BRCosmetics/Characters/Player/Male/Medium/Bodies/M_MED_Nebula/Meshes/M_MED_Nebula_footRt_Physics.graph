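# Physics asset for the M_MED_Nebula cosmetic body's footRt mesh, defining dynamic foot flap bones for both left and right sides.
# Inner, middle and outer flaps each pair a small capsule on the MSTR bone with a larger, offset capsule on the 01 bone.
# Each pair has one constraint with locked twist and narrow swing limits.
# Right flaps have a wider first swing limit than left flaps.
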PhysicsAsset /BRCosmetics/Characters/Player/Male/Medium/Bodies/M_MED_Nebula/Meshes/M_MED_Nebula_footRt_Physics
bodies {
 dyn_footFlapIn_l_01 [{shapeType=Capsule, length=2, radius=2, shapeName=dyn_footFlapIn_l_01_capsule, center=<3.3393 0 0>, rotation=<rot 90 0 0>}]
 dyn_footFlapIn_l_MSTR [{shapeType=Capsule, length=0.5, radius=0.5, shapeName=dyn_footFlapIn_l_MSTR_capsule}]
 dyn_footFlapIn_r_01 [{shapeType=Capsule, length=2, radius=2, shapeName=dyn_footFlapIn_r_01_capsule, center=<3.3393 0 0>, rotation=<rot 90 0 0>}]
 dyn_footFlapIn_r_MSTR [{shapeType=Capsule, length=0.5, radius=0.5, shapeName=dyn_footFlapIn_r_MSTR_capsule}]
 dyn_footFlapMd_l_01 [{shapeType=Capsule, length=2, radius=2, shapeName=dyn_footFlapMd_l_01_capsule, center=<3.3393 0 0>, rotation=<rot 90 0 0>}]
 dyn_footFlapMd_l_MSTR [{shapeType=Capsule, length=0.5, radius=0.5, shapeName=dyn_footFlapMd_l_MSTR_capsule}]
 dyn_footFlapMd_r_01 [{shapeType=Capsule, length=2, radius=2, shapeName=dyn_footFlapMd_r_01_capsule, center=<3.3393 0 0>, rotation=<rot 90 0 0>}]
 dyn_footFlapMd_r_MSTR [{shapeType=Capsule, length=0.5, radius=0.5, shapeName=dyn_footFlapMd_r_MSTR_capsule}]
 dyn_footFlapOt_l_01 [{shapeType=Capsule, length=2, radius=2, shapeName=dyn_footFlapOt_l_01_capsule, center=<3.3393 0 0>, rotation=<rot 90 0 0>}]
 dyn_footFlapOt_l_MSTR [{shapeType=Capsule, length=0.5, radius=0.5, shapeName=dyn_footFlapOt_l_MSTR_capsule}]
 dyn_footFlapOt_r_01 [{shapeType=Capsule, length=2, radius=2, shapeName=dyn_footFlapOt_r_01_capsule, center=<3.3393 0 0>, rotation=<rot 90 0 0>}]
 dyn_footFlapOt_r_MSTR [{shapeType=Capsule, length=0.5, radius=0.5, shapeName=dyn_footFlapOt_r_MSTR_capsule}]
}
constraints [
 {bone1Name=dyn_footFlapIn_l_01, bone2Name=dyn_footFlapIn_l_MSTR, swing1LimitDegrees=10, swing2LimitDegrees=5, twistLimitDegrees=0}
 {bone1Name=dyn_footFlapMd_l_01, bone2Name=dyn_footFlapMd_l_MSTR, swing1LimitDegrees=10, swing2LimitDegrees=5, twistLimitDegrees=0}
 {bone1Name=dyn_footFlapOt_l_01, bone2Name=dyn_footFlapOt_l_MSTR, swing1LimitDegrees=10, swing2LimitDegrees=5, twistLimitDegrees=0}
 {bone1Name=dyn_footFlapIn_r_01, bone2Name=dyn_footFlapIn_r_MSTR, swing1LimitDegrees=20, swing2LimitDegrees=5, twistLimitDegrees=0}
 {bone1Name=dyn_footFlapMd_r_01, bone2Name=dyn_footFlapMd_r_MSTR, swing1LimitDegrees=20, swing2LimitDegrees=5, twistLimitDegrees=0}
 {bone1Name=dyn_footFlapOt_r_01, bone2Name=dyn_footFlapOt_r_MSTR, swing1LimitDegrees=20, swing2LimitDegrees=5, twistLimitDegrees=0}
]
