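# Post-copy animation blueprint for the Nebula male character body, adding secondary motion to shoulder pad, feet and skirt.
# It copies the attached parent's pose, applies bone constraints and a crouch-aware Control Rig, then simulates each foot with RigidBody nodes.
# Three skirt physics setups are blended by bool for front end, stationary crouch and default.
# The update event sets gravity override and derives stationary-crouch state from crouching and movement.
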AnimBlueprint /BRCosmetics/Characters/Player/Male/Medium/Bodies/M_MED_Nebula/Meshes/M_MED_Nebula_AnimBP
target_skeleton=&/M_MED_Nebula_Skeleton
graphs [
 AnimGraph {nodes [
   n1 = RigidBody!(
    ComponentPose=Local To Component(LocalPose=Use cached pose 'feetCache'())
    bAlphaBoolEnabled:bool
    OverrideWorldGravity:Vector=Get Gravity Override()
    ExternalForce
   ) -> (Pose) {
    alphaInputType=Bool
    bOverrideWorldGravity=true
    overridePhysicsAsset=&/M_MED_Nebula_Skirt_Physics
    simulationSpace=BaseBoneSpace
    alphaBoolBlend {blendOption=Linear}
    baseBoneRef {boneName=root}
    componentAppliedLinearAccClamp=<6000 6000 6000>
    componentLinearAccScale=<0.01 0.01 0.4>
    componentLinearVelScale=<0.01 0.01 0.4>
   }
   n2 = Component To Local(ComponentPose=@n1)
   n3 = RigidBody(ComponentPose=Local To Component(LocalPose=Use cached pose 'feetCache'()), OverrideWorldGravity=Get Gravity Override()) {
    alphaInputType=Bool
    bOverrideWorldGravity=true
    overridePhysicsAsset=&/M_MED_Nebula_Skirt_Crouch_Physics
    simulationSpace=BaseBoneSpace
    alphaBoolBlend {blendOption=Linear}
    baseBoneRef {boneName=root}
    componentAppliedLinearAccClamp=<6000 6000 6000>
    componentLinearAccScale=<0.01 0.01 0.4>
    componentLinearVelScale=<0.01 0.01 0.4>
   }
   n4 = Component To Local(ComponentPose=@n3)
   n5 = Get isCrouchStoppedBool() -> (isCrouchStoppedBool:bool)
   n6 = RigidBody(ComponentPose=Local To Component(LocalPose=Use cached pose 'feetCache'()), OverrideWorldGravity=Get Gravity Override()) {
    alphaInputType=Bool
    bOverrideWorldGravity=true
    overridePhysicsAsset=&/M_MED_Nebula_Skirt_FrontEnd_Physics
    simulationSpace=BaseBoneSpace
    alphaBoolBlend {blendOption=Linear}
    baseBoneRef {boneName=root}
    componentAppliedLinearAccClamp=<6000 6000 6000>
    componentLinearAccScale=<0.01 0.01 0.4>
    componentLinearVelScale=<0.01 0.01 0.4>
   }
   Output Pose(Result=Blend Poses by bool(
     BlendPose_0=Component To Local(ComponentPose=@n6)
     BlendPose_1=Blend Poses by bool(BlendPose_0=@n4, BlendPose_1=@n2, BlendTime_0=0.5, bActiveValue=@n5) {blendType=Linear}
     BlendTime_0=0.5
     bActiveValue=Get bIsInFrontEnd()
    ) {blendType=Linear}) {layerGroup=None}
   n7 = "Constraint - dyn_shoulderPad_l_01 (PPPP)"!(
    ComponentPose=Local To Component(LocalPose=Copy Pose From Mesh() {bUseAttachedParent=true})
    Alpha=1
    ConstraintWeights_0=0.6
    ConstraintWeights_1:float=0.3
    ConstraintWeights_2:float=0.07
    ConstraintWeights_3:float=0.03
   ) -> (Pose) {
    alphaBoolBlend {blendOption=Linear}
    boneToModify {boneName=dyn_shoulderPad_l_01}
    constraintSetup [
     {targetBone {boneName=clavicle_l}}
     {targetBone {boneName=upperarm_twist_01_l}}
     {targetBone {boneName=deform_upperarm_fr_l}}
     {targetBone {boneName=deform_upperarm_bk_l}}
    ]
    constraintWeights [1,1,1,1]
   }
   n8 = "Constraint - dyn_ball_l_01 (PP)"!(ComponentPose=@n7, Alpha=1, ConstraintWeights_0=1, ConstraintWeights_1:float=1) -> (Pose) {
    alphaBoolBlend {blendOption=Linear}
    boneToModify {boneName=dyn_ball_l_01}
    constraintSetup [{targetBone {boneName=spine_01}}, {targetBone {boneName=pelvis}}]
    constraintWeights [1,1]
   }
   n9 = "Constraint - dyn_ball_l_02 (PP)"!(ComponentPose=@n8, Alpha=1, ConstraintWeights_0=1, ConstraintWeights_1:float=1) -> (Pose) {
    alphaBoolBlend {blendOption=Linear}
    boneToModify {boneName=dyn_ball_l_02}
    constraintSetup [{targetBone {boneName=spine_01}}, {targetBone {boneName=pelvis}}]
    constraintWeights [1,1]
   }
   n10 = "Constraint - dyn_ball_l_03 (PP)"!(ComponentPose=@n9, Alpha=1, ConstraintWeights_0=1, ConstraintWeights_1:float=1) -> (Pose) {
    alphaBoolBlend {blendOption=Linear}
    boneToModify {boneName=dyn_ball_l_03}
    constraintSetup [{targetBone {boneName=spine_01}}, {targetBone {boneName=pelvis}}]
    constraintWeights [1,1]
   }
   n11 = "Constraint - dyn_ball_r_01 (PP)"!(ComponentPose=@n10, Alpha=1, ConstraintWeights_0=1, ConstraintWeights_1:float=1) -> (Pose) {
    alphaBoolBlend {blendOption=Linear}
    boneToModify {boneName=dyn_ball_r_01}
    constraintSetup [{targetBone {boneName=spine_01}}, {targetBone {boneName=pelvis}}]
    constraintWeights [1,1]
   }
   n12 = "Constraint - dyn_ball_r_02 (PP)"!(ComponentPose=@n11, Alpha=1, ConstraintWeights_0=1, ConstraintWeights_1:float=1) -> (Pose) {
    alphaBoolBlend {blendOption=Linear}
    boneToModify {boneName=dyn_ball_r_02}
    constraintSetup [{targetBone {boneName=spine_01}}, {targetBone {boneName=pelvis}}]
    constraintWeights [1,1]
   }
   n13 = "Constraint - dyn_ball_r_03 (PP)"!(ComponentPose=@n12, Alpha=1, ConstraintWeights_0=1, ConstraintWeights_1:float=1) -> (Pose) {
    alphaBoolBlend {blendOption=Linear}
    boneToModify {boneName=dyn_ball_r_03}
    constraintSetup [{targetBone {boneName=spine_01}}, {targetBone {boneName=pelvis}}]
    constraintWeights [1,1]
   }
   n14 = Control Rig!(Source=Component To Local(ComponentPose=@n13), Alpha, isCrouching:double=Get isCrouchingStopped() -> (isCrouchingStopped:double)) -> (Pose) {
    alphaBoolBlend {blendOption=Linear}
    controlRigAssetReference {blueprintRigClass=&/M_MED_Nebula_ControlRig.M_MED_Nebula_ControlRig_C}
   }
   Save cached pose 'controlRigCache'(Pose=@n14) {cacheName=controlRigCache}
   n15 = RigidBody(ComponentPose=Local To Component(LocalPose=Use cached pose 'controlRigCache'()), OverrideWorldGravity=Get Gravity Override()) {
    alphaInputType=Bool
    bOverrideWorldGravity=true
    overridePhysicsAsset=&/M_MED_Nebula_footLf_Physics
    simulationSpace=BaseBoneSpace
    alphaBoolBlend {blendOption=Linear}
    baseBoneRef {boneName=foot_l}
    componentAppliedLinearAccClamp=<6000 6000 6000>
    componentLinearAccScale=<0.4 0.4 0.4>
    componentLinearVelScale=<0.4 0.4 0.4>
   }
   n16 = RigidBody(ComponentPose=@n15, OverrideWorldGravity=Get Gravity Override()) {
    alphaInputType=Bool
    bOverrideWorldGravity=true
    overridePhysicsAsset=&/M_MED_Nebula_footRt_Physics
    simulationSpace=BaseBoneSpace
    alphaBoolBlend {blendOption=Linear}
    baseBoneRef {boneName=foot_r}
    componentAppliedLinearAccClamp=<6000 6000 6000>
    componentLinearAccScale=<0.4 0.4 0.4>
    componentLinearVelScale=<0.4 0.4 0.4>
   }
   Save cached pose 'feetCache'(Pose=Component To Local(ComponentPose=@n16)) {cacheName=feetCache}
  ]}
 EventGraph {nodes [
   Event BlueprintUpdateAnimation()
   > n1 = Sequence!(execute) -> (then_0, then_1, then_2:Exec, then_3:Exec)
   > n2 = Gravity Override State Function DEPRECATED(CharacterPartAnimInstance=Self-Reference(), GameTimeInSeconds=0, GravityOverrideParams=Get Gravity Override Params())
   > Set Gravity Override(Gravity Override=@n2)
   n3 = Branch(execute=@n1.then_1, Condition=Get bIsCrouching())
   > Set isCrouchingAlpha(isCrouchingAlpha:double=1) -> (Output_Get:double)
   > n4 = Branch(Condition=Get charMoving() -> (charMoving:bool))
   > Set isCrouchingStopped(isCrouchingStopped:double=0) -> (Output_Get:double)
   > Set isCrouchStoppedBool(isCrouchStoppedBool:bool=false) -> (Output_Get:bool)
   Set isCrouchingStopped(execute=@n4.else, isCrouchingStopped=1)
   > Set isCrouchStoppedBool(isCrouchStoppedBool=true)
   Set isCrouchingAlpha(execute=@n3.else, isCrouchingAlpha=0)
   > Set isCrouchingStopped(isCrouchingStopped=0)
   > Set isCrouchStoppedBool(isCrouchStoppedBool=false)
   n5 = PlayerMovement(execute=@n1.then_2, AnimInstance=Self-Reference(), XDotProductCoefficient=1, YDotProductCoefficient=1, ZDotProductCoefficient=1)
   > n6 = Branch(Condition="Not Equal (Float)"(A=@n5.YMovement))
   > Set charMoving(charMoving:bool=true) -> (Output_Get:bool)
   Set charMoving(execute=@n6.else, charMoving=false)
   n7 = Branch(Condition="float > float"(A=MapRangeClamped(Value=0, InRangeA=0, InRangeB=1), B=0.5))
   > Set PlayerMovement(PlayerMovement:bool=true) -> (Output_Get:bool)
   Set PlayerMovement(execute=@n7.else, PlayerMovement=false)
  ]}
]
variables [
 charMoving {type=boolean}
 Gravity Override {type=Vector}
 Gravity Override Params {type=GravityOverrideParamsStruct}
 isCrouchingAlpha {type=number}
 isCrouchingStopped {type=number}
 isCrouchStoppedBool {type=boolean}
 PlayerMovement {type=boolean}
]
